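# LightArrow is a debug material in ImpostorBaker, likely for drawing a light direction arrow, judging by its name.
# It is an unlit, translucent material whose only node is a Color vector parameter, defaulting to yellow with zero alpha.
# The graph has no other math or textures and shows no link from that parameter to a material output.
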Material /ImpostorBaker/Materials/Debug/LightArrow
graph {
 properties {blendMode=BLEND_Translucent, pixelDepthOffsetMode=PDOM_Legacy, shadingModel=MSM_Unlit, usedShadingModels=Unlit}
 nodes [VectorParameter() {parameterName=Color, defaultValue=<rgba 1 1 0.1 0>}]
}
