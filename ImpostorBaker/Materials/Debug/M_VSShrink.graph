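# Debug material for Impostor Baker that visualizes hemi-octahedral frame selection and tests vertex shrinking by frame.
# It maps the actor-to-camera direction onto a hemi-octahedral grid, drawn with sphere-mask dots and numeric readouts.
# An offset branch compares that cell with each vertex normal's cell, cancelling world position on mismatch to collapse non-matching geometry.
# Parallax, dither, depth-offset and vertex-color index experiments also remain, with no material output connections shown.
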Material /ImpostorBaker/Materials/Debug/M_VSShrink
graph {
 properties {bTangentSpaceNormal=false, blendMode=BLEND_Masked, opacityMaskClipValue=0.5, pixelDepthOffsetMode=PDOM_Legacy, usedShadingModels=Default Lit}
 nodes [
  n1 = CameraPositionWS()
  n2 = ActorPositionWS()
  n3 = MaterialFunctionCall(Float3=Normalize(VectorInput=Subtract(A=@n1, B=@n2))) {materialFunction=&BreakOutFloat3Components}
  n4 = Normalize(VectorInput=MaterialFunctionCall(X=@n3.R, Y=@n3.R, Z=Max(A=@n3.B) {constB=0.0001}) {materialFunction=&MakeFloat3}.Result)
  n5 = ConstantBiasScale(None=MaterialFunctionCall!(Unit Vector=@n4) {materialFunction=/ImpostorBaker/MaterialFunctions/Internal/UnitVectorToHemiOctahedra}.Hemi Octahedron) {desc=0 to 1}
  n6 = Subtract(A=ScalarParameter() {defaultValue=8, parameterName=FramesXY})
  n7 = Floor(None=Subtract(A=Multiply(A=@n5, B=@n6)) {constB=0.5})
  n8 = Frac(None=Multiply(A=TextureCoordinate()) {constB=16})
  n9 = Floor(None=Subtract(A=Multiply(A=@n8, B=@n6)) {constB=0.5})
  n10 = ScalarParameter() {defaultValue=0.1, parameterName=FrameDebugRadius}
  n11 = Add(
   A=Multiply(A=SphereMask(A=@n7, B=@n9, Radius=@n10)) {constB=0.05}
   B=SphereMask(A=@n8, B=@n5, Radius=ScalarParameter() {defaultValue=0.005, parameterName=FrameDebugRadius2})
  )
  n12 = ComponentMask(None=VertexColor()) {g=true, r=true}
  n13 = Multiply(
   A=SphereMask(A=@n8, B=@n12, Radius=ScalarParameter() {defaultValue=0.005, parameterName=FrameDebugRadius2_1})
   B=Constant3Vector() {constant=<rgba 1 0 0 0>}
  )
  n14 = Multiply(A=Add(A=Add(A=@n11, B=@n13), B=Constant())) {constB=2}
  n15 = ScalarParameter() {defaultValue=1}
  n16 = TextureSample() {samplerType=SAMPLERTYPE_LinearColor, texture=/Engine/Functions/Engine_MaterialFunctions02/PivotPainter2/Black_1x1_EXR_Texture}
  n17 = Frac(None=Multiply(A=TextureCoordinate()) {constB=8})
  Add(
   A=Add(A=Multiply(A=@n14, B=@n15), B=Multiply(A=@n16.RGB) {constB=0.2})
   B=MaterialFunctionCall!(Vector3=VertexNormalWS(), MaximumNumberOfDigits, UVs=@n17, DebugTextLocation RG_UpperRight BA_LowerLeft, Component Spacing) {materialFunction=/Engine/Functions/Engine_MaterialFunctions02/Utility/DebugFloat3Values}.ColorCodedOutput
  )
  n18 = Multiply(A=WorldPosition().XYZ) {constB=-1}
  n19 = Normalize(VectorInput=Subtract(A=CameraPositionWS(), B=ActorPositionWS()))
  n20 = MaterialFunctionCall(Float3=@n19) {materialFunction=&BreakOutFloat3Components}
  n21 = MaterialFunctionCall(X=@n20.R, Y=@n20.R, Z=Max(A=@n20.B) {constB=0.0001}) {materialFunction=&MakeFloat3}
  n22 = ConstantBiasScale(None=MaterialFunctionCall(Unit Vector=Normalize(VectorInput=@n21.Result)) {materialFunction=&UnitVectorToHemiOctahedra}.Hemi Octahedron) {desc=0 to 1}
  n23 = Multiply(A=@n22) {constB=15}
  n24 = Divide(A=Floor(None=Subtract(A=@n23) {constB=0.5})) {constB=16}
  n25 = VertexNormalWS()
  n26 = Multiply(A=ConstantBiasScale(None=MaterialFunctionCall(Unit Vector=@n25) {materialFunction=&UnitVectorToHemiOctahedra}.Hemi Octahedron) {desc=0 to 1}) {constB=15}
  n27 = Divide(A=Floor(None=Subtract(A=@n26) {constB=0.5})) {constB=16}
  n28 = Constant() {r=1}
  n29 = Constant()
  n30 = Add(
   A=If(A=ComponentMask(None=@n24) {r=true}, B=ComponentMask(None=@n27) {r=true}, "A > B"=@n28, "A == B"=@n29, "A < B"=@n28)
   B=If(A=ComponentMask(None=@n24) {g=true}, B=ComponentMask(None=@n27) {g=true}, "A > B"=@n28, "A == B"=@n29, "A < B"=@n28)
  )
  Add(A=Multiply(A=@n18, B=Saturate(None=@n30)), B=Multiply(A=VertexNormalWS(), B=ScalarParameter() {defaultValue=5, parameterName=R}))
  AppendVector(
   A=MaterialFunctionCall!(Vector2=@n12, MaximumNumberOfDigits, UVs=@n17, DebugTextLocation RG_UpperRight BA_LowerLeft, Component Spacing) {materialFunction=/Engine/Functions/Engine_MaterialFunctions02/Utility/DebugFloat2Values}.ColorCodedOutput
   B=Constant()
  )
  n31 = TextureCoordinate()
  ConstantBiasScale(None=TextureSample(UVs=@n31) {samplerType=SAMPLERTYPE_LinearColor, texture=&Black_1x1_EXR_Texture}.RGB) {bias=-0.5, scale=2}
  Constant() {r=0.2}
  Constant()
  MaterialExpressionTextureObject_0 = TextureObject() {samplerType=SAMPLERTYPE_LinearColor, texture=&Black_1x1_EXR_Texture}
  MaterialExpressionTextureCoordinate_3 = TextureCoordinate()
  MaterialExpressionScalarParameter_7 = ScalarParameter() {defaultValue=64, parameterName=Steps}
  MaterialExpressionDivide_5 = Divide(B=ScalarParameter() {defaultValue=64, parameterName=Steps}) {desc=StepSize}
  Custom!(
   Tex=@MaterialExpressionTextureObject_0
   UV=@MaterialExpressionTextureCoordinate_3
   MaxSteps=@MaterialExpressionScalarParameter_7
   stepsize=@MaterialExpressionDivide_5
   UVDist
   InDDX
   InDDY
   HeightMapChannel
  ) {
   code="float rayheight=1;\r\nfloat oldray=1;\r\nfloat2 offset=0;\r\nfloat oldtex=1;\r\nfloat texatray;\r\nfloat yintersect;\r\nint i=0;\r\n\r\nwhile (i<MaxSteps+2)\r\n{\r\ntexatray=dot(HeightMapChannel, Tex.SampleLevel(TexSampler,UV+offset,0));\r\n\r\nif (rayheight < texatray)\r\n{\r\nfloat xintersect = (oldray-oldtex)+(texatray-rayheight);\r\nxintersect=(texatray-rayheight)/xintersect;\r\nyintersect=(oldray*(xintersect))+(rayheight*(1-xintersect));\r\noffset-=(xintersect*UVDist);\r\nbreak;\r\n}\r\n\r\noldray=rayheight;\r\nrayheight-=stepsize;\r\noffset+=UVDist;\r\noldtex=texatray;\r\n\r\n\r\ni++;\r\n}\r\n\r\nfloat3 output;\r\noutput.xy=offset;\r\noutput.z=yintersect;\r\nreturn output;"
   description=Parallax Only Mapping
   inputs [
    {inputName=Tex, input {expression=&self:MaterialExpressionTextureObject_0}}
    {inputName=UV, input {expression=&self:MaterialExpressionTextureCoordinate_3}}
    {inputName=MaxSteps, input {expression=&self:MaterialExpressionScalarParameter_7}}
    {inputName=stepsize, input {expression=&self:MaterialExpressionDivide_5}}
    {inputName=UVDist, input {expression=null}}
    {inputName=InDDX, input {expression=null}}
    {inputName=InDDY, input {expression=null}}
    {inputName=HeightMapChannel, input {expression=null}}
   ]
  }
  Divide(A=TextureCoordinate()) {constB=16}
  Divide(A=Floor(None=@n23)) {constB=16}
  n32 = Transform(None=CameraVectorWS()) {transformSourceType=TRANSFORMSOURCE_World, transformType=TRANSFORM_Tangent}
  n33 = ComponentMask(None=@n32) {g=true, r=true}
  Divide(
   A=Divide(A=Multiply(A=@n33) {constB=-1}, B=ComponentMask(None=@n32) {b=true}) {constB=1}
   B=ScalarParameter() {defaultValue=8, parameterName=FramesXY}
  )
  Frac(None=Multiply(A=@n31) {constB=8})
  n34 = VertexColor()
  If(
   A=Floor(None=Multiply(A=@n34.R) {constB=255.0001})
   B=Floor(None=ScalarParameter() {parameterName=idx})
   "A > B"=@n18
   "A == B"=Constant()
   "A < B"=@n18
  )
  MaterialFunctionCall!(
   Heightmap Texture
   Height Ratio
   Min Steps
   Max Steps
   UVs
   Heightmap Channel
   Reference Plane
   -----------------------------
   Use World Coordinates
   Specify Manual Texture Size
   Manual Texture Size
   ------------------------------
   "Render Shadows (Occlusion Mapping)"
   Light Vector
   Shadow Steps
   Shadow Penumbra
   Transform To VertexNormal
  ) {materialFunction=/Engine/Functions/Engine_MaterialFunctions01/Texturing/ParallaxOcclusionMapping}
  Multiply(A=@n4, B=Subtract(A=ScalarParameter() {defaultValue=8, parameterName=FramesXY}))
  n35 = Abs(None=DotProduct(A=VertexNormalWS(), B=@n19))
  n36 = Power(Base=@n35, Exp=ScalarParameter() {defaultValue=8})
  Multiply(A=@n16.A, B=@n36)
  Multiply(A=MaterialFunctionCall(Alpha Threshold=@n36) {materialFunction=&DitherTemporalAA}.Result) {constB=2}
  n37 = TextureSample() {samplerType=SAMPLERTYPE_LinearColor, texture=&Black_1x1_EXR_Texture}
  Multiply(A=Divide(A=OneMinus(None=@n37.A), B=@n35), B=ScalarParameter() {defaultValue=50, parameterName=pdo})
  Multiply(A=@n22) {constB=16}
 ]
}
