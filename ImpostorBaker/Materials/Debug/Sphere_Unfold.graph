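# Unlit Impostor Baker debug material that shows the octahedral or hemi-octahedral (FullSphere switch) frame grid folded onto a sphere.
# UVs decode to unit vectors that tint the fill and offset vertices to radius R, with UW blending from the unmodified mesh and Sprite toward billboarding.
# A triangulated wireframe, with diagonals flipped by quadrant, overlays the fill.
# Debug Result highlights the camera direction and its interpolation triangle.
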Material /ImpostorBaker/Materials/Debug/Sphere_Unfold
graph {
 properties {bUsedWithStaticLighting=true, pixelDepthOffsetMode=PDOM_Legacy, shadingModel=MSM_Unlit, usedShadingModels=Unlit}
 nodes [
  n1 = TextureCoordinate()
  n2 = StaticBoolParameter() {parameterName=FullSphere}
  n3 = ScalarParameter() {defaultValue=6, parameterName=XYFrames}
  n4 = Multiply(
   A=StaticSwitch(
    True=TextureCoordinate()
    False=MaterialFunctionCall(XY=@n1) {materialFunction=&Swizzle}.YX
    Value=@n2
   )
   B=@n3
  )
  n5 = Frac(None=@n4)
  n6 = MaterialFunctionCall(Float2=@n5) {materialFunction=&BreakOutFloat2Components}
  n7 = OneMinus(None=@n6.G)
  n8 = MaterialFunctionCall!(2D Index=Floor(None=TextureCoordinate() {uTiling=2, vTiling=2}), Columns X=Constant() {r=2}) {materialFunction=/Engine/Functions/Engine_MaterialFunctions01/Coordinates/2Dto1DIndex}
  n9 = Constant()
  n10 = Constant() {r=1}
  n11 = Constant()
  n12 = Constant() {r=1}
  n13 = Saturate(None=Add(A=If(A=@n8.1D Index, "A > B"=@n9, "A == B"=@n10, "A < B"=@n9), B=If(A=@n8.1D Index, "A > B"=@n11, "A == B"=@n12, "A < B"=@n11) {constB=3}))
  n14 = LinearInterpolate(A=@n13, B=OneMinus(None=@n13), Alpha=ScalarParameter() {defaultValue=1, parameterName=FlipTesselation})
  n15 = MaterialFunctionCall!(UV=LinearInterpolate(A=@n5, B=AppendVector(A=@n7, B=@n6.R), Alpha=@n14)) {materialFunction=/ImpostorBaker/MaterialFunctions/Internal/TriangleInterpolator}
  n16 = Normalize(VectorInput=Subtract(A=CameraPositionWS(), B=ActorPositionWS()))
  n17 = MaterialFunctionCall!(Vector=@n16) {materialFunction=/Engine/Functions/Engine_MaterialFunctions01/Vectors/UnitVectorToOctahedron}
  n18 = MaterialFunctionCall!(Unit Vector=@n16) {materialFunction=/ImpostorBaker/MaterialFunctions/Internal/UnitVectorToHemiOctahedra}
  n19 = Multiply(A=ConstantBiasScale(None=StaticSwitch(True=@n17.Octahedron, False=@n18.Hemi Octahedron, Value=@n2)), B=@n3)
  n20 = Frac(None=@n19)
  n21 = MaterialFunctionCall(Float2=@n20) {materialFunction=&BreakOutFloat2Components}
  n22 = LinearInterpolate(A=@n20, B=AppendVector(A=OneMinus(None=@n21.G), B=@n21.R), Alpha=@n14)
  n23 = ScalarParameter() {defaultValue=0.02, parameterName=CameraDebug}
  n24 = Multiply(
   A=LinearInterpolate(
    A=OneMinus(None=@n15.Mask)
    B=@n15.Mask
    Alpha=MaterialFunctionCall(UV=@n22) {materialFunction=&TriangleInterpolator}.Mask
   )
   B=SphereMask(A=Floor(None=@n4), B=Floor(None=@n19), Radius=@n23)
  )
  n25 = Multiply(A=@n24, B=@n15.Result)
  n26 = TextureCoordinate() {uTiling=2, vTiling=2}
  n27 = Subtract(A=TextureCoordinate() {uTiling=2, vTiling=2})
  n28 = MaterialFunctionCall!(Hemi Octahedron=MaterialFunctionCall(XYZ=@n27, XY=@n27) {materialFunction=&Swizzle}.YX) {materialFunction=/ImpostorBaker/MaterialFunctions/Internal/HemiOctahedronToUnitVector}
  n29 = StaticSwitch(
   True=MaterialFunctionCall!(Octahedron=Subtract(A=@n26)) {materialFunction=/Engine/Functions/Engine_MaterialFunctions01/Vectors/OctahedronToUnitVector}.Unit Vector
   False=@n28.Unit Vector
   Value=StaticBoolParameter() {parameterName=FullSphere}
  )
  n30 = SphereMask(A=@n16, B=@n29, Radius=@n23)
  n31 = ScalarParameter() {defaultValue=6, parameterName=XYFrames}
  n32 = Multiply(A=@n31) {constB=0.5}
  n33 = MaterialFunctionCall!(
   Texture Scale=ScalarParameter() {defaultValue=0.9965, parameterName=fasdf}
   UVs=Multiply(A=TextureCoordinate(), B=@n31)
   Offset XY Texturesize Z=MaterialFunctionCall(X=@n32, Y=@n32, Z=Constant() {r=1}) {materialFunction=&MakeFloat3}.Result
  ) {materialFunction=/Engine/Functions/Engine_MaterialFunctions01/Texturing/ScaleUVsAroundPoint}
  n34 = MaterialFunctionCall(Float2=@n33.UVs) {materialFunction=&BreakOutFloat2Components}
  n35 = OneMinus(None=@n34.G)
  n36 = Abs(None=DotProduct(A=@n29, B=CameraVectorWS()))
  n37 = LinearInterpolate(B=ScalarParameter() {defaultValue=0.3, parameterName=Powe}, Alpha=ScalarParameter() {defaultValue=1, parameterName=UW, sliderMax=1})
  n38 = MaterialFunctionCall!(
   UV=LinearInterpolate(A=@n33.UVs, B=AppendVector(A=@n35, B=@n34.R), Alpha=@n14)
   Line Width=Divide(A=ScalarParameter() {defaultValue=0.03, parameterName=LineWIdth}, B=Power(Base=@n36, Exp=@n37))
  ) {materialFunction=/ImpostorBaker/MaterialFunctions/Internal/TriangulatedWireframe}
  n39 = ScalarParameter() {defaultValue=0.5, parameterName=Wires}
  n40 = OneMinus(None=Multiply(A=@n24, B=ScalarParameter() {parameterName=Debug Result}))
  n41 = Multiply(A=ScalarParameter() {defaultValue=0.5, parameterName=FillOpacity}, B=Saturate(None=@n29))
  Add(
   A=Multiply(A=Add(A=@n25, B=@n30), B=ScalarParameter() {parameterName=Debug Result})
   B=Add(A=Multiply(A=@n38.Result, B=@n39), B=Multiply(A=@n40, B=@n41))
  )
  n42 = Multiply(A=WorldPosition().XYZ) {constB=-1}
  n43 = ScalarParameter() {defaultValue=250, parameterName=R}
  n44 = Multiply(A=@n29, B=@n43)
  LinearInterpolate(
   A=LinearInterpolate(B=Add(A=@n42, B=@n44), Alpha=ScalarParameter() {defaultValue=1, parameterName=UW, sliderMax=1})
   B=MaterialFunctionCall!(Center Location, 0-1 UVs, XY scale=Multiply(A=@n43) {constB=2}, OverridePivot, "Normals (Optional)") {materialFunction=&Sprite}.World position offset
   Alpha=ScalarParameter() {parameterName=Sprite}
  )
  Multiply(A=@n16) {constB=-1}
  Multiply(A=@n28.Unit Vector, B=VectorParameter() {defaultValue=<rgba -1 -1 -1 0>}.RGB)
  Subtract(A=@n1)
 ]
}
