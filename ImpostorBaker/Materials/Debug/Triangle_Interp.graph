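# Unlit debug material in ImpostorBaker that visualizes the TriangleInterpolator function over doubled texture coordinates.
# It sums a dimmed view of the interpolator's result with a triangulated wireframe of adjustable line width.
# A test vector parameter picks a UV point, highlighted by a SphereMask whose radius is a parameter.
# The interpolator's result at that point is also printed as color-coded digits through DebugFloat3Values.
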Material /ImpostorBaker/Materials/Debug/Triangle_Interp
graph {
 properties {bUsedWithStaticLighting=true, pixelDepthOffsetMode=PDOM_Legacy, shadingModel=MSM_Unlit, usedShadingModels=Unlit}
 nodes [
  n1 = TextureCoordinate() {uTiling=2, vTiling=2}
  n2 = Multiply(A=MaterialFunctionCall!(UV=@n1) {materialFunction=/ImpostorBaker/MaterialFunctions/Internal/TriangleInterpolator}.Result) {constB=0.25}
  n3 = ComponentMask(None=VectorParameter() {parameterName=test, defaultValue=<rgba 0.5 0.5 0 0>}.RGB) {g=true, r=true}
  n4 = MaterialFunctionCall!(
   Vector3=MaterialFunctionCall(UV=@n3) {materialFunction=&TriangleInterpolator}.Result
   MaximumNumberOfDigits
   UVs
   DebugTextLocation RG_UpperRight BA_LowerLeft
   Component Spacing
  ) {materialFunction=/Engine/Functions/Engine_MaterialFunctions02/Utility/DebugFloat3Values}
  n5 = ScalarParameter() {defaultValue=0.05, parameterName=R}
  Add(
   A=MaterialFunctionCall!(UV=@n1, Line Width=ScalarParameter() {defaultValue=0.0075, parameterName=lw}) {materialFunction=/ImpostorBaker/MaterialFunctions/Internal/TriangulatedWireframe}.Result
   B=Add(A=Add(A=@n2, B=@n4.ColorCodedOutput), B=SphereMask(A=@n1, B=@n3, Radius=@n5))
  )
 ]
}
